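# Physics asset for the Log pickaxe cosmetic, a two-handed melee weapon with ten dynamic leaf bones.
# The handle bone carries a capsule along its length and a large box near the top.
# Each leaf bone has a small capsule body constrained directly to the handle rather than chained.
# All ten constraints share the same narrow swing and twist limits, keeping leaf movement slight around the handle.
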PhysicsAsset /BRCosmetics/Weapons/FORT_Melee/Pickaxe_Log/Meshes/Log_Axe_PhysicsAsset
bodies {
 Melee_TwoHanded_Handle [
  {shapeType=Capsule, length=74.5, radius=4.5, center=<0.4463 0 14.9006>}
  {shapeType=Box, extentX=31.8887, extentY=59.9721, extentZ=35.9978, center=<-0.0744 5.7578 73.8147>}
 ]
 dyn_leaf_01 [{shapeType=Capsule, length=3.282, radius=2.7796, center=<-1.3408 -3.5825 2.548>, rotation=<rot 5.7162 -8.2195 -55.409>}]
 dyn_leaf_02 [{shapeType=Capsule, length=3.282, radius=2.2796, center=<1.5291 -0.912 4.1328>, rotation=<rot -14.2378 -7.1854 -5.565>}]
 dyn_leaf_03 [{shapeType=Capsule, length=2.032, radius=2.0296, center=<0.0455 1.1841 2.7405>, rotation=<rot 0.7322 -8.6226 9.5704>}]
 dyn_leaf_04 [{shapeType=Capsule, length=3.282, radius=2.7796, center=<1.5471 -3.7704 -2.5764>, rotation=<rot 14.5567 15.7842 -125.7608>}]
 dyn_leaf_05 [{shapeType=Capsule, length=4.168, radius=2.2672, center=<-4.0024 -1.0006 0.9939>, rotation=<rot 90 0 0>}]
 dyn_leaf_06 [{shapeType=Capsule, length=5.0596, radius=3.6973, center=<0.6318 5.1293 -2.4706>, rotation=<rot 0 0 -54.9992>}]
 dyn_leaf_07 [{shapeType=Capsule, length=4.4636, radius=2.9519, center=<2.6143 1.4423 -4.6275>, rotation=<rot 18.7472 -7.0963 -21.1712>}]
 dyn_leaf_08 [{shapeType=Capsule, length=6.1702, radius=3.6174, center=<-3.4539 1.4251 -5.8865>, rotation=<rot -14.9999 0 0>}]
 dyn_leaf_09 [{shapeType=Capsule, length=4.8333, radius=3.1904, center=<-1.3899 3.7088 -4.2671>, rotation=<rot -4.5305 2.1174 -25.0819>}]
 dyn_leaf_10 [{shapeType=Capsule, length=5.0211, radius=3.5328, center=<1.9442 -3.3981 -4.7254>, rotation=<rot 4.3287 2.5048 30.0951>}]
}
constraints [
 {bone1Name=dyn_leaf_01, bone2Name=Melee_TwoHanded_Handle, swing1LimitDegrees=20, swing2LimitDegrees=20, twistLimitDegrees=20}
 {bone1Name=dyn_leaf_02, bone2Name=Melee_TwoHanded_Handle, swing1LimitDegrees=20, swing2LimitDegrees=20, twistLimitDegrees=20}
 {bone1Name=dyn_leaf_03, bone2Name=Melee_TwoHanded_Handle, swing1LimitDegrees=20, swing2LimitDegrees=20, twistLimitDegrees=20}
 {bone1Name=dyn_leaf_04, bone2Name=Melee_TwoHanded_Handle, swing1LimitDegrees=20, swing2LimitDegrees=20, twistLimitDegrees=20}
 {bone1Name=dyn_leaf_05, bone2Name=Melee_TwoHanded_Handle, swing1LimitDegrees=20, swing2LimitDegrees=20, twistLimitDegrees=20}
 {bone1Name=dyn_leaf_06, bone2Name=Melee_TwoHanded_Handle, swing1LimitDegrees=20, swing2LimitDegrees=20, twistLimitDegrees=20}
 {bone1Name=dyn_leaf_07, bone2Name=Melee_TwoHanded_Handle, swing1LimitDegrees=20, swing2LimitDegrees=20, twistLimitDegrees=20}
 {bone1Name=dyn_leaf_08, bone2Name=Melee_TwoHanded_Handle, swing1LimitDegrees=20, swing2LimitDegrees=20, twistLimitDegrees=20}
 {bone1Name=dyn_leaf_09, bone2Name=Melee_TwoHanded_Handle, swing1LimitDegrees=20, swing2LimitDegrees=20, twistLimitDegrees=20}
 {bone1Name=dyn_leaf_10, bone2Name=Melee_TwoHanded_Handle, swing1LimitDegrees=20, swing2LimitDegrees=20, twistLimitDegrees=20}
]
